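# Sound cue that plays the reactive foley sound for the WaywardRebel character skin.
# A single wave player feeds the Char_WaywardRebel_Reactive wave straight to the output through basic foley attenuation and the player foley sound class, with boosted volume.
# A concurrency override limits it to one instance per owner, stopping the oldest when a new one starts.
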SoundCue /Game/Athena/Sounds/Character/Skin/WaywardRebel/Char_WaywardRebel_Reactive_Cue
root=WavePlayer_0
nodes [{id=WavePlayer_0, properties {soundWaveAssetPtr=&/Char_WaywardRebel_Reactive}}]
properties {
 attenuationSettings=&Foley_Attenuation_Basic_Att
 bOverrideConcurrency=true
 maxDistance=1536
 soundClassObject=&Fort_Player_Foley
 volumeMultiplier=3.5
 concurrencyOverrides {bLimitToOwner=true, maxCount=1, resolutionRule=StopOldest}
}
